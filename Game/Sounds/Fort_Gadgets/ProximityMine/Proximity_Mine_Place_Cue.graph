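# Proximity Mine placement sound cue for Fort gadgets, playing a single wave, Proximity_New_Arm, through one wave player at the root.
# It routes to the Fort_Gadgets_Items sound class and uses the shared gadget-placement attenuation.
# A concurrency override caps simultaneous instances at a few, stopping the oldest first, and debug is enabled on the cue.
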SoundCue /Game/Sounds/Fort_Gadgets/ProximityMine/Proximity_Mine_Place_Cue
root=WavePlayer_0
nodes [{id=WavePlayer_0, properties {soundWaveAssetPtr=&/Proximity_New_Arm}}]
properties {
 attenuationSettings=/Game/Sounds/Attenuation/Fort_GadgetPlace_Att
 bDebug=true
 bOverrideConcurrency=true
 maxDistance=2000
 soundClassObject=&Fort_Gadgets_Items
 concurrencyOverrides {maxCount=3, resolutionRule=StopOldest}
}
