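# Sound cue that plays the proximity mine activation trigger sound for the Fortnite gadget.
# A single wave player feeds the Proximity_New_Trigger wave straight to the output, with no randomization or modulation.
# It uses gadget-activation attenuation and the gadgets sound class, and overrides concurrency to prevent new instances from starting once the limit is reached.
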SoundCue /Game/Sounds/Fort_Gadgets/ProximityMine/fort_proximity_mine_activate_Cue
root=WavePlayer_0
nodes [{id=WavePlayer_0, properties {soundWaveAssetPtr=&/Proximity_New_Trigger}}]
properties {
 attenuationSettings=&Fort_GadgetActivate_Att
 bOverrideConcurrency=true
 maxDistance=4000
 soundClassObject=/Game/Sounds/FortSoundClassesAndModes/Fort_Gadgets_StW
 concurrencyOverrides {resolutionRule=PreventNew}
}
